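# Physics asset for the Bad Egg backpack cosmetic, giving its skeletal mesh simple collision and joint limits for physics-driven motion.
# Capsule bodies cover the pack bone and two dynamic missile bones.
# Each missile is constrained to the pack with matching swing and twist limits, so both can wobble within a narrow cone while staying attached.
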PhysicsAsset /BRCosmetics/Accessories/FORT_Backpacks/Backpack_M_MED_Bad_Egg/Meshes/M_MED_Bad_Egg_Physics
bodies {
 dyn_missile_01 [{shapeType=Capsule, length=25, radius=4, center=<-6 0 4.5>, rotation=<rot 90 0 0>}]
 dyn_missile_02 [{shapeType=Capsule, length=25.2536, radius=4.7, center=<-3 0 4.5>, rotation=<rot 90 0 0>}]
 pack [{shapeType=Capsule, length=13.3174, radius=7.6035, center=<-0.5493 -0.5455 -8.2025>, rotation=<rot 0 -91.8633 -48.5649>}]
}
constraints [
 {bone1Name=dyn_missile_01, bone2Name=pack, swing1LimitDegrees=25, swing2LimitDegrees=15, twistLimitDegrees=10}
 {bone1Name=dyn_missile_02, bone2Name=pack, swing1LimitDegrees=25, swing2LimitDegrees=15, twistLimitDegrees=10}
]
